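# Translucent unlit material for the balloon burner flame, with optional Battle Bus animation and "Drive the Bus" behavior.
# Panning noise, Fresnel and UV gradient feed a fire-ramp lookup and opacity; the DepthFade switch also enables hue shift.
# Layered sine waves wobble vertices through WPO.
# Drive the Bus blends custom primitive data over time, tints emissive and, by static switch, rotates vertices from object orientation, weighted by local height.
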Material /Game/Missions/Primary/LaunchTheBalloon/Materials/M_BalloonBurner_Flame
graph {
 properties {
  bUseMaterialAttributes=true
  bUsedWithInstancedStaticMeshes=true
  blendMode=BLEND_Translucent
  pixelDepthOffsetMode=PDOM_Legacy
  shadingModel=MSM_Unlit
  usedShadingModels=Unlit
 }
 nodes [
  ComponentMask(None=TextureCoordinate() {vTiling=6}) {g=true}
  Multiply(A=LocalPosition().XYZ, B=Constant3Vector() {constant=<rgba 1 1 0.1 0>})
  n1 = NamedRerouteUsage()
  NamedRerouteDeclaration(None=Saturate(None=Abs(None=@n1))) {name=DriveTheBus, nodeColor=<rgba 0 0.2266 1 1>}
  n2 = ScalarParameter() {bUseCustomPrimitiveData=true, desc=Drive the Bus, parameterName=CPD_BalloonBendAmount_Sides_A, primitiveDataIndex=1, sliderMax=1}
  n3 = ScalarParameter() {bUseCustomPrimitiveData=true, desc=Drive the Bus, parameterName=CPD_BalloonBendAmount_Sides_B, primitiveDataIndex=2, sliderMax=1}
  n4 = ScalarParameter() {defaultValue=0.5, desc=Overriten By BP on construct B_StockBattleBus, group=WPO, parameterName=DP_BankingRefreshRate, sliderMax=1}
  n5 = Divide(A=Subtract(A=Time(), B=ScalarParameter() {bUseCustomPrimitiveData=true, parameterName=CPD_Gametime, sliderMax=1}), B=@n4)
  NamedRerouteDeclaration(None=LinearInterpolate(A=@n2, B=@n3, Alpha=Saturate(None=@n5))) {name=CPD_DriveTheBus, nodeColor=<rgba 0 1 0.6016 1>}
  PerInstanceRandom()
  n6 = Multiply(A=NamedRerouteUsage(), B=Constant() {r=5})
  n7 = Frac(None=Time())
  n8 = Multiply(A=Constant() {r=7.1}, B=ScalarParameter() {defaultValue=1, parameterName=TimeSpeed})
  n9 = OneMinus(None=ComponentMask(None=TextureCoordinate()) {g=true})
  n10 = Multiply(A=@n9, B=@n9)
  n11 = Sine(None=Add(A=Multiply(A=@n7, B=@n8) {constB=7.1}, B=Multiply(A=@n10) {constB=-2}))
  n12 = ScalarParameter() {defaultValue=1.2, parameterName=FineAmp}
  n13 = TextureCoordinate()
  n14 = Sine(None=Add(A=Time(), B=ComponentMask(None=@n13) {g=true})) {period=1.31}
  n15 = ScalarParameter() {defaultValue=4, parameterName=MiddleAmp}
  n16 = Time()
  n17 = Multiply(A=Constant() {r=1.32}, B=ScalarParameter() {defaultValue=1, parameterName=TimeSpeed})
  n18 = ComponentMask(None=TextureCoordinate()) {g=true}
  n19 = Multiply(A=Constant() {r=0.63}, B=ScalarParameter() {defaultValue=1, parameterName=TimeSpeed})
  n20 = Sine(None=Add(A=Multiply(A=@n16, B=@n17) {constB=1.31}, B=Multiply(A=@n18, B=@n19) {constB=0.63})) {period=1.31}
  n21 = Add(
   A=Multiply(A=Add(A=@n6), B=Multiply(A=@n11, B=@n12) {constB=1.2, desc=Amplitude})
   B=Add(A=Multiply(A=@n14, B=@n15) {constB=4, desc=Amplitude}, B=Multiply(A=@n20, B=@n15) {constB=4, desc=Amplitude})
  )
  n22 = Time()
  n23 = Constant() {r=4}
  n24 = ScalarParameter() {defaultValue=1, parameterName=TimeSpeed}
  n25 = AppendVector(A=ComponentMask(None=WorldPosition().XYZ) {b=true}, B=Constant2Vector())
  n26 = Add(A=Multiply(A=Frac(None=@n22), B=Multiply(A=@n23, B=@n24)) {constB=4}, B=Frac(None=Multiply(A=@n25) {constB=-0.02}))
  n27 = ScalarParameter() {defaultValue=3, parameterName=LargeAmp}
  n28 = ComponentMask(None=Multiply(A=Sine(None=@n26), B=@n27) {desc=Amplitude}) {r=true}
  n29 = ComponentMask(None=Multiply(A=Cosine(None=@n26), B=@n27)) {r=true}
  n30 = Constant()
  n31 = OneMinus(None=ComponentMask(None=TextureCoordinate()) {g=true})
  n32 = Add(
   A=Multiply(A=VertexNormalWS(), B=Max(A=@n21) {constB=-2.5})
   B=Multiply(
    A=MaterialFunctionCall(X=@n28, Y=@n29, Z=@n30) {materialFunction=&MakeFloat3}.Result
    B=Multiply(A=@n31, B=@n31)
   )
  ) {constB=0}
  n33 = TextureCoordinate()
  n34 = Clamp(None=Add(A=ComponentMask(None=@n33) {g=true}) {constB=0.5})
  n35 = Constant2Vector()
  n36 = Multiply(A=Sine(None=Time()) {period=0.5}) {constB=0.79}
  n37 = Multiply(A=Sine(None=Time()) {period=0.73}) {constB=0.5}
  n38 = TextureCoordinate()
  n39 = Multiply(
   A=Multiply(A=Add(A=@n36, B=@n37), B=ScalarParameter() {parameterName=UpAndDownMovement}) {constB=5}
   B=ComponentMask(None=OneMinus(None=@n38)) {g=true}
  )
  n40 = Multiply(A=ScalarParameter() {defaultValue=1, parameterName=MaxWPO}, B=Add(A=Multiply(A=@n32, B=@n34), B=AppendVector(A=@n35, B=@n39)))
  n41 = Add(A=MaterialFunctionCall() {materialFunction=/Game/Athena/Aircraft/PartyBus_RT/Materials/MF_PartyBusAnimationWPO}.Result, B=@n40)
  n42 = TextureCoordinate()
  n43 = VectorParameter() {parameterName=MainColor, defaultValue=<rgba 0.63 0.7502 1 1>}
  n44 = MaterialFunctionCall(Exponent=ScalarParameter() {defaultValue=2.5, parameterName=FresnelExponent}) {materialFunction=&MF_Fresnel_FX}
  n45 = VectorParameter() {parameterName=InnerColor, defaultValue=<rgba 1 1 1 1>}
  n46 = VectorParameter() {parameterName=MiddleColor, defaultValue=<rgba 0.0768 0.3055 0.67 1>}
  n47 = ComponentMask(None=OneMinus(None=TextureCoordinate())) {g=true}
  n48 = ScalarParameter() {defaultValue=1.5, parameterName=FresnelExp}
  n49 = ComponentMask(None=TextureCoordinate()) {g=true}
  n50 = Add(A=CameraVectorWS(), B=Constant3Vector() {constant=<rgba 0 0 -0.5 0>})
  n51 = Multiply(A=MaterialFunctionCall(Exponent=Multiply(A=@n48, B=@n49), Camera Vector=Normalize(VectorInput=@n50)) {materialFunction=&MF_Fresnel_FX}.Fresnel Inverted)
  n52 = Clamp(None=Add(A=OneMinus(None=@n49)) {constB=-0.75})
  n53 = Multiply(A=Multiply(A=@n52, B=@n52)) {constB=4}
  n54 = Panner(
   Coordinate=TextureCoordinate() {uTiling=2, vTiling=0.6}
   Speed=AppendVector(A=ScalarParameter() {defaultValue=0.17, parameterName=XSpeed}, B=ScalarParameter() {defaultValue=2.5, parameterName=YSpeed})
  ) {bFractionalPart=true, speedX=0.17, speedY=2.5}
  n55 = TextureSampleParameter2D(UVs=@n54) {parameterName=MainTexture, texture=&T_NoisePack_01}
  n56 = StaticSwitchParameter(
   True=Multiply(
    A=@n55.R
    B=TextureSampleParameter2D(UVs=@n54) {parameterName=MaskTexture, texture=/Game/Effects/Fort_Effects/Textures/Masks/T_Vert_Gradient_Softer}.G
   )
   False=@n55.R
  ) {parameterName=UseMaskTexture?}
  n57 = Clamp(None=Add(A=Add(A=@n51, B=@n53) {constB=0}, B=Multiply(A=@n56) {constB=0.25}))
  n58 = TextureSample(UVs=MaterialFunctionCall(X=@n57, Y=Constant() {r=0.5}) {materialFunction=&MakeFloat2}.Result) {texture=/Game/Missions/Primary/LaunchTheBalloon/Textures/T_Balloon_FireRamp}
  n59 = Add(
   A=Multiply(A=ComponentMask(None=@n42) {g=true}, B=Multiply(A=@n43.RGB, B=@n44.Fresnel Inverted))
   B=Multiply(A=LinearInterpolate(A=@n45.RGB, B=@n46.RGB, Alpha=@n47), B=Multiply(A=@n58.RGB) {constB=10})
  )
  n60 = MaterialFunctionCall(Hue Shift Percentage=ScalarParameter() {parameterName=HueShift}, Texture=@n59) {materialFunction=&HueShift}
  n61 = OneMinus(None=@n57)
  n62 = Clamp(None=Add(A=Power(Base=@n57) {constExponent=8}, B=Power(Base=@n61) {constExponent=4}) {constB=0})
  n63 = StaticSwitchParameter(True=DepthFade(Opacity=@n62, FadeDistance=ScalarParameter() {defaultValue=30, parameterName=DepthFade}), False=@n62) {parameterName=DepthFade}
  n64 = ScalarParameter() {defaultValue=1, parameterName=OpacityMult}
  n65 = GetMaterialAttributes(None=SetMaterialAttributes!(
    MaterialAttributes
    Emissive Color=StaticSwitchParameter(True=@n60.Result, False=@n59) {parameterName=DepthFade}
    Opacity=Multiply(A=@n63, B=@n64)
    World Position Offset=StaticSwitchParameter(True=@n41, False=@n40) {parameterName=UseBusAnim}
   ) {attributeSetTypes [#1, #2, #3]}) {attributeGetTypes [#1, #3]}
  n66 = Multiply(A=@n65.Emissive Color) {constB=0.1}
  n67 = Saturate(None=Multiply(A=NamedRerouteUsage()) {constB=5})
  n68 = Multiply(A=VectorParameter() {parameterName=MiddleColor_DriveTheBus, defaultValue=<rgba 0 1.3218 2 1>}.RGB, B=@n67)
  n69 = Constant3Vector() {constant=<rgba 0 0 1 0>}
  n70 = ObjectOrientation()
  n71 = Divide(A=Arccosine!(None=DotProduct(A=@n69, B=@n70))) {constB=6.2832}
  n72 = CrossProduct(A=@n70, B=@n69)
  n73 = Constant3Vector() {constant=<rgba 0 0 0 0>}
  n74 = Transform(None=Normalize(VectorInput=Add(A=@n72, B=@n73) {constB=0.0001})) {transformSourceType=TRANSFORMSOURCE_World, transformType=TRANSFORM_Local}
  n75 = Constant3Vector()
  n76 = LocalPosition()
  n77 = LocalPosition()
  n78 = Constant() {r=140}
  n79 = Multiply(
   A=Transform(None=MaterialFunctionCall!(RotationAngle=@n71, NormalizedRotationAxis=@n74, PivotPoint=@n75, Position=@n76.XYZ, Normal, TangentU, TangentV) {materialFunction=/Game/Characters/Player/Male/Medium/Bodies/M_MED_Inferno/FX/Mat/RotateVertexAboutAxis}.WPO) {transformSourceType=TRANSFORMSOURCE_Local}
   B=Saturate(None=Divide(A=@n77.Z, B=@n78) {constB=140})
  )
  SetMaterialAttributes!(
   MaterialAttributes=@n65.MaterialAttributes
   Emissive Color=LinearInterpolate(A=@n65.Emissive Color, B=Add(A=@n66, B=@n68), Alpha=@n67)
   World Position Offset=StaticSwitchParameter(True=Add(A=@n65.World Position Offset, B=@n79), False=@n65.World Position Offset) {defaultValue=true, parameterName=UseDriveTheBus}
  ) {attributeSetTypes [#1, #3]}
 ]
}
