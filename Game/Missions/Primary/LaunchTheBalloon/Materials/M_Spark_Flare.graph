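# Additive, unlit particle sprite material for a star-shaped spark flare in the Launch the Balloon primary mission.
# It multiplies the RGB of the spark star texture by particle color, so the emitter controls the tint.
# The result feeds the Emissive input of the shared MF_FX_EmissiveAdjustments material function.
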Material /Game/Missions/Primary/LaunchTheBalloon/Materials/M_Spark_Flare
graph {
 properties {bUsedWithParticleSprites=true, blendMode=BLEND_Additive, pixelDepthOffsetMode=PDOM_Legacy, shadingModel=MSM_Unlit, usedShadingModels=Unlit}
 nodes [MaterialFunctionCall(Emissive=Multiply(A=TextureSample() {texture=/Game/Missions/Primary/LaunchTheBalloon/Textures/T_Spark_Star}.RGB, B=ParticleColor().RGB)) {materialFunction=&MF_FX_EmissiveAdjustments}]
}
